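# Sound cue for the Creative match flush tray notification UI sound.
# A single wave player plays the matchmaking game flush notification wave, slightly attenuated, through the application tray notification sound class.
# It uses the shared tray notification concurrency set, with overrides specifying one instance and stop-oldest resolution.
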SoundCue /Game/Sounds/Fort_UI_Sounds/TrayNotifications/SC_TrayNotification_MatchFlush_Creative
root=WavePlayer_27
nodes [{id=WavePlayer_27, properties {soundWaveAssetPtr=&/UI_MatchMaking_GameFlush_TrayNotification_01}}]
properties {
 maxDistance=2097152
 soundClassObject=/Game/Sounds/FortSoundClassesAndModes/Fort_Application_TrayNotification
 volumeMultiplier=0.9
 concurrencyOverrides {maxCount=1, resolutionRule=StopOldest}
 concurrencySet [&/CONC_TrayNotification]
}
